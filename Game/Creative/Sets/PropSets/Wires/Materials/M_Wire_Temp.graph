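# M_Wire_Temp is a material in the Creative Wires prop set, apparently a temporary placeholder judging by its name.
# It uses the Default Lit shading model and holds a single dark navy Constant3Vector colour node.
# The graph shows no textures, parameters or connection from that colour to a material output.
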Material /Game/Creative/Sets/PropSets/Wires/Materials/M_Wire_Temp
graph {
 properties {pixelDepthOffsetMode=PDOM_Legacy, usedShadingModels=Default Lit}
 nodes [Constant3Vector() {constant=<rgba 0.0205 0.0279 0.0729 0>}]
}
